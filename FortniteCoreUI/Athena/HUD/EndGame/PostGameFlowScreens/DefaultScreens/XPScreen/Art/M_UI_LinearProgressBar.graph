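# Translucent UI material for a linear progress bar, stored with the post-game XP screen art.
# Animated start-to-end progress, optionally snapped to segments, drives a hard-edged or SDF rounded-box fill mask, colored by a three-color gradient over an empty color.
# An optional delta band animates separately with decaying color.
# Pixel-sized segment gaps and rounded outline masks shape the result, with features toggled by 0–1 scalar lerps rather than static switches.
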Material /FortniteCoreUI/Athena/HUD/EndGame/PostGameFlowScreens/DefaultScreens/XPScreen/Art/M_UI_LinearProgressBar
graph {
 properties {
  bUsedWithStaticMesh=false
  blendMode=BLEND_Translucent
  materialDomain=MD_UI
  pixelDepthOffsetMode=PDOM_Legacy
  usedShadingModels=Default Lit
 }
 nodes [
  n1 = VectorParameter() {
   desc="Color for empty part of progress bar.\r\nAlpha defines opacity for empty part of progress bar."
   group=Color
   parameterName=EmptyColor
   sortPriority=10
   defaultValue=<rgba 0.3924 0.3924 0.3924 0.2>
  }
  n2 = VectorParameter() {desc=Start color in gradient, group=Color, parameterName=ColorA, sortPriority=1, defaultValue=<rgba 0 0 1 1>}
  n3 = VectorParameter() {desc=Central color in gradient, group=Color, parameterName=ColorB, sortPriority=2, defaultValue=<rgba 0 1 0 1>}
  n4 = VectorParameter() {desc=End color in gradient, group=Color, parameterName=ColorC, sortPriority=3, defaultValue=<rgba 1 0 0 1>}
  n5 = ComponentMask(None=NamedRerouteUsage()) {r=true}
  n6 = VectorParameter() {desc=Color for delta part of progress bar., group=Delta, parameterName=DeltaColor, sortPriority=20, defaultValue=<rgba 1 0 0.829 1>}
  n7 = VectorParameter() {group=Delta, parameterName=DeltaColorDecay, sortPriority=21, defaultValue=<rgba 0 1 1 1>}
  n8 = ScalarParameter() {
   desc="Property for animating Delta color decay\r\n(from DeltaColor to DeltaColorDecay)\r\n0 - 1"
   group=Delta
   parameterName=AnimateDeltaColor
   sliderMax=1
   sortPriority=33
  }
  n9 = ScalarParameter() {desc="Enables delta functionality for progress bar.\r\nbool", group=Delta, parameterName=ShowDelta, sliderMax=1, sortPriority=1}
  n10 = Multiply(A=NamedRerouteUsage(), B=@n9)
  LinearInterpolate(
   A=@n1.RGB
   B=LinearInterpolate(
    A=MaterialFunctionCall(A=@n2.RGB, B=@n3.RGB, C=@n4.RGB, Alpha=@n5) {materialFunction=&3ColorBlend}.Result
    B=LinearInterpolate(A=@n6.RGB, B=@n7.RGB, Alpha=@n8)
    Alpha=@n10
   ) {constAlpha=0}
   Alpha=Max(A=@n10, B=NamedRerouteUsage())
  )
  n11 = NamedRerouteUsage()
  n12 = VectorParameter() {
   desc="Color for empty part of progress bar.\r\nAlpha defines opacity for empty part of progress bar."
   group=Color
   parameterName=EmptyColor
   sortPriority=10
   defaultValue=<rgba 0.3924 0.3924 0.3924 0.2>
  }
  n13 = ScalarParameter() {defaultValue=1, desc="Alpha for filled part of progress bar.\r\n0 -1", group=Color, parameterName=FillAlpha, sliderMax=1}
  n14 = NamedRerouteUsage()
  n15 = ScalarParameter() {desc="Enables delta functionality for progress bar.\r\nbool", group=Delta, parameterName=ShowDelta, sliderMax=1, sortPriority=1}
  n16 = LinearInterpolate(A=@n12.A, B=@n13, Alpha=Max(A=Multiply(A=@n14, B=@n15), B=NamedRerouteUsage()))
  Multiply(
   A=LinearInterpolate(
    A=@n16
    B=Multiply(A=@n16, B=@n11)
    Alpha=ScalarParameter() {desc="Enable segments functionality.\r\nbool", group=Segments, parameterName=UseSegments, sliderMax=1, sortPriority=1}
   )
   B=NamedRerouteUsage()
  )
  n17 = ScalarParameter() {group=Progress_Layout, parameterName=ProgressEdgeSharpness, sliderMax=1, sortPriority=2}
  n18 = ScalarParameter() {
   desc="When progress value changes, it defines where change starts.\r\n0 - 1"
   group=Progress
   parameterName=ProgressStart
   sliderMax=1
   sortPriority=1
  }
  n19 = ScalarParameter() {
   defaultValue=1
   desc="When progress value changes, it defines where change ends.\r\n0 - 1"
   group=Progress
   parameterName=ProgressEnd
   sliderMax=1
   sortPriority=2
  }
  n20 = ScalarParameter() {
   defaultValue=0.45
   desc="Property for animating progress value change (from ProgressStart to ProgressEnd)\r\n0 - 1"
   group=Progress
   parameterName=AnimateProgress
   sliderMax=1
   sortPriority=3
  }
  n21 = LinearInterpolate(A=@n18, B=@n19, Alpha=@n20)
  n22 = ScalarParameter() {defaultValue=8, desc=Number of segments in progress bar, group=Segments, parameterName=SegmentsCount, sortPriority=2}
  n23 = LinearInterpolate(
   A=@n21
   B=MaterialFunctionCall!("Value (0-1)"=@n21, Steps=Floor(None=@n22)) {materialFunction=/FortUILibrary/MaterialFunctions/Basics/MF_UI_StepRatio}.Stepped Ratio
   Alpha=ScalarParameter() {desc="Change progress value only with segments chanks.\r\nbool", group=Segments, parameterName=SnapToSegments, sliderMax=1}
  )
  n24 = ComponentMask(None=Add(A=NamedRerouteUsage(), B=NamedRerouteDeclaration(None=@n23) {name=Progress, nodeColor=<rgba 0 0 0 1>}) {constB=0.5}) {r=true}
  n25 = ScalarParameter() {
   defaultValue=1
   desc="Will use rounded box to show progress value. Not support delta functionality and snap to segments.\r\nbool"
   group=Progress_Layout
   parameterName=UseRoundedProgress
   sliderMax=1
  }
  n26 = LinearInterpolate(A=Saturate(None=SmoothStep(Max=@n17, Value=@n24) {constMax=0.01}), B=NamedRerouteUsage(), Alpha=@n25)
  n27 = NamedRerouteUsage()
  n28 = ScalarParameter() {
   desc="Will add outline for progress bar.\r\nDoesn't work correct with \"snap to segments\".\r\nbool"
   group=Outline
   parameterName=UseOutline
   sliderMax=1
   sortPriority=1
  }
  NamedRerouteDeclaration(None=LinearInterpolate(A=@n26, B=Multiply(A=@n26, B=@n27), Alpha=@n28)) {name=ProgressMask, nodeColor=<rgba 0 0 0 1>}
  n29 = TextureCoordinate()
  NamedRerouteDeclaration(None=@n29) {name=UVs, nodeColor=<rgba 0 0 0 1>}
  NamedRerouteDeclaration(None=@n26) {name=ProgressMaskForDelta, nodeColor=<rgba 0 0 0 1>}
  n30 = Add(A=Divide(A=Add(A=@n29) {constB=-0.5}, B=NamedRerouteUsage())) {constB=0.5}
  NamedRerouteDeclaration(None=Multiply(A=@n30) {constB=-1}) {name=RemappedUVs, nodeColor=<rgba 0 0 0 1>}
  NamedRerouteDeclaration(None=@n30) {name=UVsWithoutOutline, nodeColor=<rgba 0 0 0 1>}
  n31 = ScalarParameter() {defaultValue=4, desc=Horizontal size of gap between segements in px., group=Segments, parameterName=SeparatorThickness_PX, sortPriority=3}
  n32 = NamedRerouteDeclaration(None=MaterialFunctionCall!(Value PX=@n31, UV Pixel Size, X or Y) {materialFunction=/FortUILibrary/MaterialFunctions/Basics/MF_UI_PixelsToNormalized}."Normalized (>1px)") {name=GapSizeNormalized, nodeColor=<rgba 0 0 0 1>}
  n33 = Add(A=NamedRerouteUsage(), B=@n32) {constA=1}
  n34 = ScalarParameter() {defaultValue=8, desc=Number of segments in progress bar, group=Segments, parameterName=SegmentsCount, sliderMin=1, sortPriority=2}
  n35 = Floor(None=@n34)
  n36 = Max(A=@n35)
  NamedRerouteDeclaration(None=Subtract(A=Divide(A=@n33, B=@n36), B=@n32)) {name=SegmentSizeNormalized, nodeColor=<rgba 0 0 0 1>}
  NamedRerouteDeclaration(None=Add(A=@n35) {constB=-1}) {name=GapsCount, nodeColor=<rgba 0 0 0 1>}
  n37 = ScalarParameter() {
   desc="Will add outline for progress bar.\r\nDoesn't work correct with \"snap to segments\".\r\nbool"
   group=Outline
   parameterName=UseOutline
   sliderMax=1
   sortPriority=1
  }
  n38 = LinearInterpolate(
   B=ScalarParameter() {defaultValue=10, desc=Size of outline for rounded progress bar in px., group=Outline, parameterName=Outline_PX}
   Alpha=@n37
  )
  n39 = Saturate(None=Multiply(A=MaterialFunctionCall(Value PX=@n38) {materialFunction=&MF_UI_PixelsToNormalized}.Normalized) {constB=2})
  NamedRerouteDeclaration(None=@n39) {name=OutlineHorNormalized, nodeColor=<rgba 0 0 0 1>}
  n40 = Multiply(A=ScalarParameter() {defaultValue=0.05, group=Progress_Layout, parameterName=LinearSDF_Sharpness, sliderMax=1, sortPriority=2}) {constB=0.5}
  n41 = NamedRerouteUsage()
  n42 = NamedRerouteUsage()
  n43 = NamedRerouteUsage()
  n44 = Add(A=ComponentMask(None=@n41) {r=true}, B=Multiply(A=OneMinus(None=@n42), B=@n43))
  n45 = ComponentMask(None=@n41) {g=true}
  n46 = NamedRerouteUsage()
  n47 = SmoothStep(
   Min=Multiply(A=@n40) {constB=-1}
   Max=@n40
   Value=MaterialFunctionCall(UVs=AppendVector(A=@n44, B=@n45), Size=AppendVector(A=@n43, B=@n46), CornerRoundness=@n46, KeepAspectRatio=StaticBool() {value=true}) {materialFunction=&MF_UI_Box}.SDF
  ) {constMax=0.01}
  NamedRerouteDeclaration(None=Multiply(A=OneMinus(None=@n47), B=NamedRerouteUsage())) {name=RoundedProgressMask, nodeColor=<rgba 0 0 0 1>}
  n48 = Multiply(A=ScalarParameter() {defaultValue=0.05, group=Progress_Layout, parameterName=LinearSDF_Sharpness, sliderMax=1, sortPriority=2}) {constB=0.5}
  n49 = Multiply(A=@n48) {constB=-1}
  n50 = ScalarParameter() {
   defaultValue=1
   desc="Will use rounded box to show progress value. Not support delta functionality and snap to segments.\r\nbool"
   group=Progress_Layout
   parameterName=UseRoundedProgress
   sliderMax=1
  }
  n51 = MaterialFunctionCall(
   Size=Constant2Vector() {g=1, r=1}
   CornerRoundness=LinearInterpolate(
    A=ScalarParameter() {desc="Value for corner roundness. \r\n0 - no roundness\r\n1 - circle", group=Progress_Layout, parameterName=CornerRoundness}
    Alpha=@n50
   )
   OutlineThickness=NamedRerouteUsage()
   KeepAspectRatio=StaticBool() {value=true}
  ) {materialFunction=&MF_UI_Box}
  NamedRerouteDeclaration(None=OneMinus(None=SmoothStep(Min=@n49, Max=@n48, Value=@n51.SDF) {constMax=0.01})) {name=RoundedMask, nodeColor=<rgba 0 0 0 1>}
  n52 = Constant() {r=1}
  n53 = Saturate(None=Multiply(A=MaterialFunctionCall(Value PX=@n38, X or Y=@n52) {materialFunction=&MF_UI_PixelsToNormalized}.Normalized) {constB=2})
  NamedRerouteDeclaration(None=OneMinus(None=@n53)) {name=BarWithoutOutlineVertNormalized, nodeColor=<rgba 0 0 0 1>}
  NamedRerouteDeclaration(None=OneMinus(None=@n39)) {name=BarWithoutOutlineHorNormalized, nodeColor=<rgba 0 0 0 1>}
  NamedRerouteDeclaration(None=@n53) {name=OutlineVertNormalized, nodeColor=<rgba 0 0 0 1>}
  NamedRerouteDeclaration(None=SmoothStep(Min=@n49, Max=@n48, Value=@n51.SDFOutline) {constMax=0.01}) {name=OutlineMask, nodeColor=<rgba 0 0 0 1>}
  n54 = ScalarParameter() {group=Progress_Layout, parameterName=ProgressEdgeSharpness, sliderMax=1, sortPriority=2}
  n55 = ScalarParameter() {
   desc="When progress value changes, it defines where change starts.\r\n0 - 1"
   group=Progress
   parameterName=ProgressStart
   sliderMax=1
   sortPriority=1
  }
  n56 = ScalarParameter() {
   defaultValue=1
   desc="When progress value changes, it defines where change ends.\r\n0 - 1"
   group=Progress
   parameterName=ProgressEnd
   sliderMax=1
   sortPriority=2
  }
  n57 = ScalarParameter() {
   defaultValue=1
   desc="Property for animating Delta value change (from ProgressStart to ProgressEnd)\r\n0 - 1"
   group=Delta
   parameterName=AnimateDelta
   sliderMax=1
  }
  n58 = LinearInterpolate(A=@n55, B=@n56, Alpha=@n57)
  n59 = MaterialFunctionCall(
   "Value (0-1)"=@n58
   Steps=Floor(None=ScalarParameter() {defaultValue=8, desc=Number of segments in progress bar, group=Segments, parameterName=SegmentsCount, sortPriority=2})
  ) {materialFunction=&MF_UI_StepRatio}
  n60 = ScalarParameter() {desc="Change progress value only with segments chanks.\r\nbool", group=Segments, parameterName=SnapToSegments, sliderMax=1}
  n61 = ComponentMask(None=Add(A=NamedRerouteUsage(), B=LinearInterpolate(A=@n58, B=@n59.Stepped Ratio, Alpha=@n60))) {r=true}
  n62 = Multiply(A=Saturate(None=SmoothStep(Max=@n54, Value=@n61) {constMax=0.01}), B=OneMinus(None=NamedRerouteUsage()))
  n63 = NamedRerouteUsage()
  n64 = ScalarParameter() {
   desc="Will add outline for progress bar.\r\nDoesn't work correct with \"snap to segments\".\r\nbool"
   group=Outline
   parameterName=UseOutline
   sliderMax=1
   sortPriority=1
  }
  NamedRerouteDeclaration(None=LinearInterpolate(A=@n62, B=Multiply(A=@n62, B=@n63), Alpha=@n64)) {name=DeltaMask, nodeColor=<rgba 0 0 0 1>}
  n65 = NamedRerouteUsage()
  n66 = Multiply(A=NamedRerouteUsage(), B=@n65) {constB=3.2}
  n67 = NamedRerouteUsage()
  n68 = Subtract(A=NamedRerouteUsage(), B=@n67)
  n69 = ComponentMask(None=NamedRerouteUsage()) {r=true}
  n70 = Add(A=@n67, B=Multiply(A=NamedRerouteUsage()) {constB=0.5})
  n71 = Subtract(A=@n69, B=@n70)
  n72 = Frac(None=Multiply(A=@n65, B=Divide(A=@n71, B=@n68) {constB=0.85}) {constA=5})
  n73 = Max(A=Ceil(None=Multiply(A=@n71) {constB=-1}), B=Floor(None=Add(A=@n69, B=@n70)))
  NamedRerouteDeclaration(None=Step(Y=Divide(A=@n66, B=@n68), X=Max(A=@n72, B=@n73))) {name=SegmentsMask, nodeColor=<rgba 0 0 0 1>}
 ]
}
